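# Niagara system for the radial smoke effect of the TaperSlink emote.
# One LargeWake emitter runs three spawn groups, gated by User.BurstIndex as spawn probability, with group-tagged particles feeding a mesh and two sprite renderers.
# Curves animate size, color, fade and material parameters, with User inputs for size and offset, time-of-day glow tuning and a front-end brightness boost.
# A BlendSpace step switches world/local position blend by an owner-speed threshold.
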NiagaraSystem /BRCosmetics/Animation/Game/MainPlayer/Emotes/TaperSlink/FX/NS_Emote_TaperSlink_RadialSmoke
dependencies {
 usedDataInterfaces [&NiagaraDataInterfaceCurve, &NiagaraDataInterfaceColorCurve]
 usedDynamicInputs [&FNTimeOfDayColor]
 usedModules [
  &SystemState
  &BlendSpaceHelper
  &self:SystemSpawnScript.SystemScriptSource.SystemScriptGraph.NiagaraNodeAssignment_4.SetVariables_#1
 ]
}
emitters {LargeWake {
  emitterUpdateScript [
   &EmitterState(Inactive Response=::NewEnumerator1, Life Cycle Mode=::NewEnumerator1, Loop Delay=0.4, Loop Duration=1, UseLoopDelay=false)
   &SpawnRate(Spawn Probability=$User.BurstIndex, SpawnRate=10, Use Spawn Probability=true) {module=SpawnRate001}
   &SpawnRate(Spawn Probability=$User.BurstIndex, SpawnGroup=1, SpawnRate=12, Use Spawn Probability=true)
   &SpawnRate(Spawn Probability=$User.BurstIndex, SpawnGroup=2, SpawnRate=8, Use Spawn Probability=true) {module=SpawnRate002}
  ]
  particleSpawnScript [
   /SetVariables(Particles.VisibilityTag=&SpawnGroup()) {isSetParameters=true}
   /SetVariables(Particles.Random=&RandomRangeFloat(Minimum=0, Maximum=1)) {isSetParameters=true}
   &InitializeParticle(
    Color=$System.SmokeCol
    Lifetime=&MergeHelper_FloatSwitch(
     0=&RandomRangeFloat(Minimum=0.7, Maximum=1.2)
     1=&RandomRangeFloat(Minimum=0.75, Maximum=1)
     2=1
     3=?
     Index=&SpawnGroup()
     Num=::NewEnumerator1
    )
    Mesh Scale=&Multiply_VectorByFloat(
     Vector=&RandomRangeVector(Minimum=<1 1 0.5>, Maximum=<1 1 1>)
     Float=&RandomRangeFloat(Minimum=0.65, Maximum=1.1)
    )
    Mesh Scale Mode=::NewEnumerator4
    Position Offset=&AddVectorToPosition(Position=<0 -25 0>, Vector=&MakeVector(X=$User.Pos_Offset_X, Y=$User.Pos_Offset_Y, Z=0.25))
    Position Offset Coordinate Space=::Local
    Sprite Rotation Mode=::NewEnumerator2
    Sprite Size Mode=::NewEnumerator3
    Uniform Sprite Size=&MergeHelper_FloatSwitch(
     0=0
     1=&Lerp_Float(A=120, B=170, Alpha=$Particles.Random)
     2=&Lerp_Float(A=250, B=300, Alpha=$Particles.Random)
     3=?
     Index=&SpawnGroup()
     Num=::NewEnumerator1
    )
    UsePositionOffset=true
    Write Material Random=true
   )
   &InitialMeshOrientation(
    Mesh Orientation Mode=::NewEnumerator2
    Orientation Coordinate Space=::World
    Rotation=&MergeHelper_VectorSwitch(
     0=&MakeVector(
      X=&MakeCustomFloatFromBool(Bool=&RandomBool(), True Float=0, False Float=0)
      Y=&MakeCustomFloatFromBool(Bool=&RandomBool(), True Float=0, False Float=0)
      Z=&RandomRangeFloat(Minimum=-1, Maximum=1)
     )
     1=<0 -0.25 0>
     2=?
     3=?
     Index=&SpawnGroup()
     Num=::NewEnumerator0
    )
    Use Rotation Vector=true
   )
   /SetVariables(Particles.CameraOffset_Center=30) {isSetParameters=true}
   /SetVariables(Particles.SpriteFacing_Glow:Vector3f=<0 0 1>) {isSetParameters=true}
  ]
  particleUpdateScript [
   &ParticleState()
   &SolveForcesAndVelocity(Clamp Velocity=false, Limit Acceleration=false)
   &DynamicMaterialParameters(
    Index 0 Param 1=&FloatFromCurve(FloatCurve=DataInterface() {propertyValues {Curve {keys [
         <at 0 0.2 RCIM_Cubic>
         {interpMode=RCIM_Cubic, tangentMode=RCTM_User, time=0.3, value=0.45}
         {interpMode=RCIM_Cubic, tangentMode=RCTM_User, time=1, value=0}
        ]}}})
    Module.Param0WriteEnabled=true
    Module.Param1WriteEnabled=true
    Module.Param2WriteEnabled=true
    Module.Param3WriteEnabled=true
   )
   &ScaleMeshSize(Scale Factor=&Multiply_VectorByFloat(
     Vector=&MakeVector(
      X=$User.SizeMult
      Y=$User.SizeMult
      Z=&FloatFromCurve(FloatCurve=DataInterface() {propertyValues {Curve {keys [<at 0.15 1>,<at 0.6 0>]}}})
     )
     Float=&FloatFromCurve(
      FloatCurve=DataInterface() {propertyValues {Curve {keys [
          {arriveTangent=2.3149, interpMode=RCIM_Cubic, leaveTangent=2.3149, tangentMode=RCTM_User, time=0, value=0}
          {arriveTangent=0.9995, interpMode=RCIM_Cubic, leaveTangent=0.9995, tangentMode=RCTM_User, time=0.3, value=0.45}
          {arriveTangent=-0.0405, interpMode=RCIM_Cubic, leaveTangent=-0.0405, tangentMode=RCTM_User, time=1, value=1}
         ]}}}
      Scale Curve=2
     )
    ))
   &ScaleColor(
    Scale Alpha=&FloatFromCurve(
     FloatCurve=DataInterface() {propertyValues {Curve {keys [
         {interpMode=RCIM_Cubic, tangentMode=RCTM_User, time=0, value=0.2}
         {interpMode=RCIM_Cubic, leaveTangent=-2.4875, tangentMode=RCTM_Break, time=0.17, value=1}
         {tangentMode=RCTM_User, time=1, value=0}
        ]}}}
     Scale Curve=0.5
    )
    Scale RGB=&Multiply_VectorByFloat(
     Vector=&VectorFromFloat(Value=&FloatFromCurve(FloatCurve=DataInterface() {propertyValues {Curve {keys [<at 0 1>,<at 1 1>]}}}, Scale Curve=0.5))
     Float=&FN_FrontEndFloatModifier(Gameplay=1, Front End=2)
    )
    ScaleA=true
    ScaleRGB=true
   )
   &ScaleSpriteSize(
    Uniform Curve Scale=$User.SizeMult
    Uniform Curve Sprite Scale=DataInterface() {propertyValues {Curve {keys [
        {arriveTangent=1.6815, interpMode=RCIM_Cubic, leaveTangent=1.6815, tangentMode=RCTM_User, time=0.2, value=0.5}
        {arriveTangent=0.1063, interpMode=RCIM_Cubic, leaveTangent=0.1063, tangentMode=RCTM_User, time=0.55, value=1}
       ]}}}
   )
   &AlignSpriteToMeshOrientation(SetAlignment=true, SetFacing=true)
   /SetVariables(
    Particles.Color_Center:LinearColor=&MultiplyLinearColor(
     Linear Color=$Particles.Initial.Color
     Scale Factor=&MakeVector4FromVectorAndFloat(
      "Vector (RGB)"=&MakeVectorFromLinearColorRGB(Color=&ColorFromCurve(ColorCurve=DataInterface() {propertyValues {
          AlphaCurve {keys [<at 0 0>,<at 0.4487 1 RCIM_Cubic>]}
          BlueCurve {keys [
            <at 0.1811 0.6425 RCIM_Cubic>
            {arriveTangent=-0.3057, interpMode=RCIM_Cubic, leaveTangent=-0.3057, time=0.4806, value=1}
            <at 0.7837 0.4583>
           ]}
          GreenCurve {keys [
            <at 0.1811 0.6562 RCIM_Cubic>
            {arriveTangent=-0.3285, interpMode=RCIM_Cubic, leaveTangent=-0.3285, time=0.4806, value=1}
            <at 0.7837 0.4583>
           ]}
          RedCurve {keys [
            <at 0.1811 0.6388 RCIM_Cubic>
            {arriveTangent=-0.2996, interpMode=RCIM_Cubic, leaveTangent=-0.2996, time=0.4806, value=1}
            <at 0.7837 0.4583>
           ]}
         }}))
      "Float (Alpha)"=&FloatFromCurve(FloatCurve=DataInterface() {propertyValues {Curve {keys [<at 0 0>, {tangentMode=RCTM_User, time=0.75, value=1}, {tangentMode=RCTM_User, time=1, value=0}]}}})
     )
    )
    Particles.DynamicMaterialParameter_Center:Vector4f=&MakeVector4(
     X=&FloatFromCurve(FloatCurve=DataInterface() {propertyValues {Curve {keys [
          {arriveTangent=1.7661, interpMode=RCIM_Cubic, leaveTangent=1.7661, tangentMode=RCTM_User, time=0, value=0}
          {arriveTangent=0.0224, interpMode=RCIM_Cubic, leaveTangent=0.0224, tangentMode=RCTM_User, time=1, value=1}
         ]}}})
     Y=1
     Z=0
     W=0
    )
   ) {isSetParameters=true}
   /SetVariables(
    Particles.Color_Glow:LinearColor=&MultiplyLinearColor(
     Linear Color=&ColorFromCurve(ColorCurve=DataInterface() {propertyValues {
        AlphaCurve {keys [<at 0 1 RCIM_Cubic>,<at 1 0 RCIM_Cubic>]}
        BlueCurve {keys [<at 0 1 RCIM_Cubic>,<at 1 1 RCIM_Cubic>]}
        GreenCurve {keys [<at 0 0.684 RCIM_Cubic>,<at 1 0.0302 RCIM_Cubic>]}
        RedCurve {keys [<at 0 0.526 RCIM_Cubic>,<at 1 0.111 RCIM_Cubic>]}
       }})
     Scale Factor=&MakeVector4FromVectorAndFloat(
      "Vector (RGB)"=&FN_Select_Vector_from_Time_Of_Day(Morning=<1 1 1>, Day=<1 1 1>, Evening=<1 1 1>, Night=<0.7 0.85 0.8>)
      "Float (Alpha)"=&FNTimeOfDayScalar(Night Scale=0.07, Morning Scale=0.06, Day Scale=0.08, Evening Scale=0.05)
     )
    )
    Particles.DynamicMaterialParameter_Glow:Vector4f=<3 0 0 0>
   ) {isSetParameters=true}
   &BlendSpace("Position Blend (World <-> Local)":NiagaraFloat=&MakeCustomFloatFromBool(
     Bool=&SetBoolByFloatComparison(A=&VectorLength(Vector=$Engine.Owner.Velocity), B=200)
     True Float=1
     False Float=0
     Translation Contribution:NiagaraFloat=1
     Rotation Contribution:NiagaraFloat=0
    ))
  ]
  rendererClasses [&NiagaraMeshRendererProperties, &NiagaraSpriteRendererProperties]
  renderers [
   {rendererClass=&NiagaraMeshRendererProperties, rendererIndex=0}
   {rendererClass=&NiagaraSpriteRendererProperties, rendererIndex=1}
   {rendererClass=&NiagaraSpriteRendererProperties, rendererIndex=2}
  ]
 }}
summary {
 systemName=NS_Emote_TaperSlink_RadialSmoke
 emitters [{emitterName=LargeWake, rendererClasses [&NiagaraMeshRendererProperties, &NiagaraSpriteRendererProperties]}]
 userVariables [
  User.bIsFrontEnd {type=NiagaraBool}
  User.bIsFrontEndPreview {type=NiagaraBool}
  User.bIsInWater {type=NiagaraBool}
  User.BurstIndex {type=NiagaraFloat, defaultValue=1, description=""}
  User.JetActiveCurve {type=NiagaraFloat, defaultValue=5, description=""}
  User.Pos_Offset_X {type=NiagaraFloat, defaultValue=0, description=""}
  User.Pos_Offset_Y {type=NiagaraFloat, defaultValue=0, description=""}
  User.SizeMult {type=NiagaraFloat, defaultValue=1, description=""}
 ]
}
