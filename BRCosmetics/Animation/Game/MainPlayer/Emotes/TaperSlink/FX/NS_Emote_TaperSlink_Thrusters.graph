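# Niagara effect for the TaperSlink emote's thrusters, with flames, glows, navigation-style blinkers and motion trails attached to mesh sockets.
# Looping mesh bursts form the flames, while flickering sprites and particle lights add glow, with sprite width adjusted by camera angle.
# Two sprites blink green and red in alternation from engine time.
# Wing and smear anim-trail ribbons respond to time of day; user parameters control flame size, height and trail life.
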NiagaraSystem /BRCosmetics/Animation/Game/MainPlayer/Emotes/TaperSlink/FX/NS_Emote_TaperSlink_Thrusters
dependencies {
 usedDataInterfaces [
  &NiagaraDataInterfaceCurve
  &NiagaraDataInterfaceColorCurve
  &NiagaraDataInterfaceArrayColor
  &NiagaraDataInterfaceVector2DCurve
  &NiagaraDataInterfaceCamera
  &NiagaraDataInterfaceSocketReader
  &NiagaraDataInterfaceSkeletalMesh
 ]
 usedDynamicInputs [&FNTimeOfDayColor]
 usedModules [
  &self:SystemSpawnScript.SystemScriptSource.SystemScriptGraph.NiagaraNodeAssignment_2.SetVariables_#1
  &self:SystemSpawnScript.SystemScriptSource.SystemScriptGraph.NiagaraNodeAssignment_4.SetVariables_#2
  &self:SystemSpawnScript.SystemScriptSource.SystemScriptGraph.NiagaraNodeAssignment_0.SetVariables_#3
  &self:SystemSpawnScript.SystemScriptSource.SystemScriptGraph.NiagaraNodeAssignment_1.SetVariables_#4
  &self:SystemSpawnScript.SystemScriptSource.SystemScriptGraph.NiagaraNodeAssignment_3.SetVariables_#5
  &SystemState
  &self:SystemSpawnScript.SystemScriptSource.SystemScriptGraph.NiagaraNodeAssignment_6.SetVariables_#6
 ]
}
emitters {
 Blinking_Lights {
  emitterUpdateScript [
   &EmitterState(
    Inactive Response=::NewEnumerator1
    Life Cycle Mode=::NewEnumerator1
    Loop Behavior=::NewEnumerator1
    Loop Duration Mode=::NewEnumerator1
    UseLoopDelay=false
   )
   &SpawnBurst_Instantaneous(Spawn Count=2, Use Spawn Probability=false, UseLoopCountLimit=false) {module=SpawnBurst_Instantaneous001}
  ]
  particleSpawnScript [
   &PartitionParticles()
   &InitializeParticle(
    Color=&SelectLinearColorFromArray(
     Array Sampling Mode=::NewEnumerator1
     Direct Array Index=$Particles.PartitionParticles.Partition
     Interpolation=&MergeHelper_FloatSwitch(0=0, 1=1, 2=?, 3=?, Index=&SpawnGroup(), Num=::NewEnumerator0)
     Color Selection Array=DataInterface() {propertyValues {ColorData [<rgba 0.1807 1 0 0.3>, <rgba 1 0.0833 0 0.3>]}}
     Random Seed=?
    )
    Sprite Rotation Mode=::NewEnumerator2
    Sprite Size Mode=::NewEnumerator3
    Uniform Sprite Size=45
    UsePositionOffset=false
    Write Material Random=true
   ) {module=InitializeParticle001}
   &CameraOffset(Camera Offset Amount=10)
  ]
  particleUpdateScript [
   &ParticleState(Kill Particles When Lifetime Has Elapsed=false)
   &SocketLocation(
    SamplePosOffset=false
    Socket Index=$Particles.PartitionParticles.Partition
    Socket Reader=$System.Socket Reader_Glow
    Socket Sampling Mode=::NewEnumerator3
    VelocitySpeedLimit=false
   )
   &SolveForcesAndVelocity(Clamp Velocity=false, Limit Acceleration=false)
   &ScaleColor(
    Scale Alpha=&MergeHelper_FloatSwitch(
     0=HlslExpression() {hlslExpression="1-step((sin(Engine.Time * 9.4248) + 1.0) * 0.5, 0.5)"}
     1=HlslExpression() {hlslExpression="step((sin(Engine.Time * 9.4248) + 1.0) * 0.5, 0.5)"}
     2=?
     3=?
     Index=$Particles.PartitionParticles.Partition
     Num=::NewEnumerator0
    )
    ScaleA=true
    ScaleRGB=false
   )
  ]
  rendererClasses [&NiagaraSpriteRendererProperties]
  renderers [{rendererClass=&NiagaraSpriteRendererProperties}]
 }
 Comet_Flame_Mesh {
  emitterUpdateScript [
   &EmitterState(Life Cycle Mode=::NewEnumerator1, Loop Delay=0.2, Loop Duration=0.1, UseLoopDelay=false)
   &SpawnBurst_Instantaneous(Spawn Count=2, Use Spawn Probability=false)
   &SpawnBurst_Instantaneous(Spawn Count=2, Spawn Group=1, Use Spawn Probability=false) {module=SpawnBurst_Instantaneous001}
  ]
  particleSpawnScript [
   &PartitionParticles()
   &InitializeParticle(
    Color=$System.Thrusters_Color_Glow
    Lifetime=0.275
    Mass Mode=::NewEnumerator1
    Mesh Scale=<-0.04 0.035 0.03>
    Mesh Scale Mode=::NewEnumerator4
    Position Offset=<0 0 50>
    UsePositionOffset=false
    Write Material Random=true
   ) {module=InitializeParticle001}
   &InitialMeshOrientation(Mesh Axis To Orient=<0 1 0>, Mesh Orientation Mode=::NewEnumerator2, Orientation Vector=<-1 0 0>, Use Rotation Vector=false)
   /SetVariables(Particles.Color_2:LinearColor=$System.Thrusters_Color, Particles.Scale_2:Vector3f=<0.035 0.032 0.025>) {isSetParameters=true}
   &DynamicMaterialParameters(
    Index 0 Param 1=5
    Index 0 Param 2=5
    Module.Index0Float0Write=true
    Module.Index0Float1Write=true
    Module.Index0Float2Write=true
    Module.Index0Float3Write=true
   )
   /SetVariables(Particles.VisibilityTag=&SpawnGroup()) {isSetParameters=true}
  ]
  particleUpdateScript [
   &SkeletalMeshLocation(
    Bone / Socket Index=$Particles.PartitionParticles.Partition
    MaskSpawnGroup=false
    Mesh Sampling Type=::NewEnumerator5
    Sampled Position Offset=&MergeHelper_VectorSwitch(0=<0 -1 -1>, 1=<0 2.5 -1>, 2=?, 3=?, Index=&SpawnGroup(), Num=::NewEnumerator0)
    Skeletal Mesh=$System.Skeletal Mesh Reader_Thrusters
    Socket Sampling Mode=::NewEnumerator5
   )
   &ParticleState()
   &ScaleColor(
    Scale Alpha=&Lerp_Float(A=1, B=0.5, Alpha=$Particles.NormalizedAge)
    Scale RGB=&MakeVectorFromLinearColorRGB(Color=&ColorFromCurve(ColorCurve=DataInterface() {propertyValues {
        AlphaCurve {keys []}
        BlueCurve {keys [<at 0 1 RCIM_Cubic>,<at 0.492 1>,<at 1 1 RCIM_Cubic>]}
        GreenCurve {keys [<at 0 1 RCIM_Cubic>,<at 0.492 1>,<at 1 1 RCIM_Cubic>]}
        RedCurve {keys [<at 0 0.25 RCIM_Cubic>,<at 0.492 0.5>,<at 1 0.875 RCIM_Cubic>]}
       }}))
    ScaleA=true
    ScaleRGB=true
   )
   &ScaleMeshSize(Scale Factor=&MakeVector(
     X=&Multiply_Float(
      A=&Multiply_Float(A=$User.Flame_Size, B=1.6)
      B=&FloatFromCurve(
       FloatCurve=DataInterface() {propertyValues {Curve {keys [<at 0 0.75 RCIM_Cubic>, {arriveTangent=-0.25, interpMode=RCIM_Cubic, leaveTangent=-0.25, time=0.5, value=2.5}, <at 1 0.5 RCIM_Cubic>]}}}
       Scale Curve=$User.Thruster_Height_Mult
      )
     )
     Y=&Multiply_Float(
      A=&Multiply_Float(A=$User.Flame_Size, B=1.25)
      B=&FloatFromCurve(FloatCurve=DataInterface() {propertyValues {Curve {keys [<at 0.5 1 RCIM_Cubic>,<at 1 1.75 RCIM_Cubic>]}}})
     )
     Z=&Multiply_Float(
      A=&Multiply_Float(A=$User.Flame_Size, B=1.25)
      B=&FloatFromCurve(FloatCurve=DataInterface() {propertyValues {Curve {keys [<at 0.5 1 RCIM_Cubic>,<at 1 1.75 RCIM_Cubic>]}}})
     )
    ))
   /SetVariables(Particles.Scale_2:Vector3f=&Multiply_Vector(
     A=&Multiply_VectorByFloat(Vector=$Particles.Initial.Scale_2, Float=$User.Thruster_Height_Mult)
     B=&MakeVector(
      X=&FloatFromCurve(
       FloatCurve=DataInterface() {propertyValues {Curve {keys [<at 0 0.75 RCIM_Cubic>, {arriveTangent=-0.25, interpMode=RCIM_Cubic, leaveTangent=-0.25, time=0.5, value=2.5}, <at 1 0.5 RCIM_Cubic>]}}}
       Scale Curve=&Multiply_Float(A=$User.Flame_Size, B=1.5)
      )
      Y=&FloatFromCurve(
       FloatCurve=DataInterface() {propertyValues {Curve {keys [<at 0.5 1 RCIM_Cubic>,<at 1 1.75 RCIM_Cubic>]}}}
       Scale Curve=&Multiply_Float(A=$User.Flame_Size, B=1.5)
      )
      Z=&FloatFromCurve(
       FloatCurve=DataInterface() {propertyValues {Curve {keys [<at 0.5 1 RCIM_Cubic>,<at 1 1.75 RCIM_Cubic>]}}}
       Scale Curve=&Multiply_Float(A=$User.Flame_Size, B=1.5)
      )
     )
    )) {isSetParameters=true}
  ]
  rendererClasses [&NiagaraMeshRendererProperties]
  renderers [
   {rendererClass=&NiagaraMeshRendererProperties, rendererIndex=0}
   {rendererClass=&NiagaraMeshRendererProperties, rendererIndex=1}
  ]
 }
 Light {
  emitterUpdateScript [
   &EmitterState(Life Cycle Mode=::NewEnumerator1, Loop Behavior=::NewEnumerator1, Loop Duration=1, UseLoopDelay=false)
   &SpawnBurst_Instantaneous(Use Spawn Probability=false, UseLoopCountLimit=false)
   &SpawnBurst_Instantaneous(Spawn Count=2, Spawn Group=1, Use Spawn Probability=false, UseLoopCountLimit=false) {module=SpawnBurst_Instantaneous001}
  ]
  particleSpawnScript [
   &PartitionParticles()
   /SetVariables(Particles.VisibilityTag=&SpawnGroup()) {isSetParameters=true}
   &InitializeParticle(
    Color=<rgba 0.5 2.75 5 0.115>
    Lifetime=10
    Position Offset=<0 20 10>
    Sprite Rotation Mode=::NewEnumerator0
    Sprite Size=<50 130>
    Sprite Size Mode=::NewEnumerator4
    UsePositionOffset=false
    Write Material Random=true
   )
   &SpriteFacingAndAlignment(Sprite Facing=<0 0 -1>, WriteAlignment=true, WriteFacing=true)
   &CameraOffset(Camera Offset Amount=2)
   /SetVariables(
    Particles.CameraOffset_Glow_Center=5
    Particles.Color_Glow_Center:LinearColor=<rgba 2.2435 6.1217 10 0.5>
    Particles.SpriteRotation_Glow_Center:NiagaraFloat=&RandomRangeFloat(Minimum=0, Maximum=360)
    Particles.SpriteSize_Glow_Center:Vector2f=&Vector2DFromFloat(Value=&RandomRangeFloat(Minimum=20, Maximum=25))
   ) {isSetParameters=true}
   /SetVariables(Particles.Color_BG:LinearColor=&MakeLinearColorFromVectorAndFloat(
     "Vector (RGB)"=&MakeVectorFromLinearColorRGB(Color=<rgba 0.167 0 0.649 0.2>)
     "Float (Alpha)"=&FN_FrontEndFloatModifier(Gameplay=0.2, Front End=0.275)
    )) {isSetParameters=true}
  ]
  particleUpdateScript [
   &SkeletalMeshLocation(
    Bone / Socket Index=$Particles.PartitionParticles.Partition
    MaskSpawnGroup=false
    Mesh Sampling Type=::NewEnumerator5
    Skeletal Mesh=$System.Skeletal Mesh Reader_Thrusters
    Socket Sampling Mode=::NewEnumerator5
   )
   &ParticleState()
   &ScaleSpriteSize(
    Scale Factor=&Vector2DFromCurve(
     Vector2Curve=DataInterface() {propertyValues {XCurve {keys [<at 0 0>,<at 1 1>]}, YCurve {keys [<at 1 1>]}}}
     CurveIndex=&Multiply_Float(
      A=1
      B=&RemapRange(
       Input Value To Remap=&Abs_Float(Float:NiagaraFloat=&DotProduct(
         Vector A=&ReturnCameraVector(Camera Vectors=::NewEnumerator4, Coordinate Space=::Local, Origin Position=$Particles.Position)
         Vector B=$Particles.SpriteFacing
        ))
       Input Min=0
       Input Max=1
       Output Min=0.5
       Output Max=1
      )
     )
     ScaleCurve=<1 1>
    )
    Scale Sprite Size Mode=::NewEnumerator2
   ) {module=ScaleSpriteSize001}
   &ScaleSpriteSize(
    Scale Sprite Size Mode=::NewEnumerator0
    Uniform Scale Factor=&Waveform(
     "Waveform [1]"=::NewEnumerator12
     "[1] Frequency"=20
     "[1] Amplitude Min/Max"=&Multiply_Vector2D_ByFloat(Float=&Multiply_Float(A=$User.Thruster_Height_Mult, B=$User.Flame_Size), Vector2D=<0.85 1>)
    )
   )
   &FN_Light_Attributes(
    Color=<rgba 1.2 6.6 12 1>
    Create Color Attribute=true
    Create Position Attribute=true
    Create Radius Attribute=true
    Enable Particle Light Enable=true
    Light Enabled=true
    Position=<0 35 20>
    Radius=&Waveform(
     "Waveform [1]"=::NewEnumerator12
     "[1] Frequency"=20
     "[1] Amplitude Min/Max"=&Multiply_Vector2D_ByFloat(Float=$User.Thruster_Height_Mult, Vector2D=<30 50>)
    )
   )
   /SetVariables(
    Particles.Position_Center_Glow:NiagaraPosition=&AddVectorToPosition(Position=$Particles.Position, Vector=<0 4 0>)
    Particles.SpriteSize_Glow_Center:Vector2f=&Multiply_Vector2D_ByFloat(
     Float=&Waveform(
      "Waveform [1]"=::NewEnumerator12
      "[1] Frequency"=20
      "[1] Amplitude Min/Max"=&Multiply_Vector2D_ByFloat(Float=&Multiply_Float(A=$User.Thruster_Height_Mult, B=$User.Flame_Size), Vector2D=<0.6 1.3>)
     )
     Vector2D=$Particles.Initial.SpriteSize_Glow_Center
    )
   ) {isSetParameters=true}
   /SetVariables(Particles.SpriteSize_BG:Vector2f=&Multiply_Vector2D(A=$Particles.SpriteSize, B=<1.3 1.55>)) {isSetParameters=true}
  ]
  rendererClasses [&NiagaraLightRendererProperties, &NiagaraSpriteRendererProperties]
  renderers [
   {rendererClass=&NiagaraLightRendererProperties, rendererIndex=0}
   {rendererClass=&NiagaraSpriteRendererProperties, rendererIndex=1}
   {rendererClass=&NiagaraSpriteRendererProperties, rendererIndex=2}
   {rendererClass=&NiagaraSpriteRendererProperties, rendererIndex=3}
  ]
 }
 Trail_Smear {
  emitterUpdateScript [
   &EmitterState(Life Cycle Mode=::NewEnumerator1, Loop Duration=0, UseLoopDelay=false)
   &FN_SpawnAnimTrail(Number of Anim Trails=2, Spawn Interval=0)
  ]
  particleSpawnScript [
   &FN_InitializeAnimTrailBoneIndecies()
   &FN_InitializeAnimTrailBoneIndecies(AnimTrail Selection=1, First Bone or Socket Index=2, Second Bone or Socket Index=3) {module=FN_InitializeAnimTrailBoneIndecies001}
   &FN_InitializeAnimTrail(
    Color=<rgba 1 0.9366 0.849 1>
    Lifetime=&Multiply_Float(A=$User.Trail_Life, B=0.35)
    Socket Reader:NiagaraDataInterfaceSocketReader=$System.Socket Reader_Smear
   )
   &DynamicMaterialParameters(Module.Index0Float0Write=true, Module.Index0Float1Write=true, Module.Index0Float2Write=true, Module.Index0Float3Write=true)
  ]
  particleUpdateScript [
   &FN_UpdateAnimTrail(Socket Reader:NiagaraDataInterfaceSocketReader=$System.Socket Reader_Smear)
   &ScaleColor(
    Scale Alpha=&FloatFromCurve(FloatCurve=DataInterface() {propertyValues {Curve {defaultValue=1, keys [{interpMode=RCIM_Cubic, leaveTangent=-3.6487, tangentMode=RCTM_Break, time=0.15, value=1}, <at 1 0>]}}})
    Scale RGB=&Multiply_Vector(
     A=&MakeVectorFromLinearColorRGB(Color=&ColorFromCurve(ColorCurve=DataInterface() {propertyValues {
         AlphaCurve {keys [<at 0 1>]}
         BlueCurve {keys [<at 0 0.5073>,<at 0.6986 0.4448>]}
         GreenCurve {keys [<at 0 0.0534>,<at 0.6986 0.0505>]}
         RedCurve {keys [<at 0 0.1139>,<at 0.6986 0.1031>]}
        }}))
     B=&MakeVectorFromLinearColorRGB(Color=$System.TOD)
    )
    ScaleA=true
    ScaleRGB=true
   )
  ]
  rendererClasses [&NiagaraRibbonRendererProperties]
  renderers [{rendererClass=&NiagaraRibbonRendererProperties}]
 }
 Trail_Wings {
  emitterUpdateScript [
   &EmitterState(Life Cycle Mode=::NewEnumerator1, Loop Duration=0, UseLoopDelay=false)
   &FN_SpawnAnimTrail(Number of Anim Trails=2)
  ]
  particleSpawnScript [
   &FN_InitializeAnimTrailBoneIndecies(First Bone or Socket Index=1)
   &FN_InitializeAnimTrail(
    AnimTrailWidth=4
    Lifetime=&Multiply_Float(A=$User.Trail_Life, B=0.675)
    Socket Reader:NiagaraDataInterfaceSocketReader=$System.Socket Reader_Tip
   )
  ]
  particleUpdateScript [
   &FN_UpdateAnimTrail(Socket Reader:NiagaraDataInterfaceSocketReader=$System.Socket Reader_Tip) {module=FN_UpdateAnimTrail001}
   &ScaleRibbonWidth(Ribbon Width Scale=&FloatFromCurve(FloatCurve=DataInterface() {propertyValues {Curve {keys [
         {arriveTangent=-1.2579, interpMode=RCIM_Cubic, time=0, value=1}
         {arriveTangent=-0.9664, interpMode=RCIM_Cubic, leaveTangent=-0.9664, tangentMode=RCTM_User, time=1, value=0}
        ]}}}))
   &ScaleColor(
    Scale Alpha=&Multiply_Float(
     A=&FloatFromCurve(
      FloatCurve=DataInterface() {propertyValues {Curve {keys [
          {interpMode=RCIM_Cubic, leaveTangent=-2.3646, tangentMode=RCTM_Break, time=0, value=1}
          <at 0.6689 0.0233>
          {arriveTangent=-0.2139, interpMode=RCIM_Cubic, tangentMode=RCTM_Break, time=1, value=0}
         ]}}}
      Scale Curve=0.3
     )
     B=&FNTimeOfDayScalar(Night Scale=1, Morning Scale=0.5, Evening Scale=0.5)
    )
    ScaleA=true
    ScaleRGB=true
   ) {module=ScaleColor002}
  ]
  rendererClasses [&NiagaraRibbonRendererProperties]
  renderers [{rendererClass=&NiagaraRibbonRendererProperties}]
 }
}
summary {
 systemName=NS_Emote_TaperSlink_Thrusters
 emitters [
  {emitterName=Trail_Wings, rendererClasses [&NiagaraRibbonRendererProperties]}
  {emitterName=Trail_Smear, rendererClasses [&NiagaraRibbonRendererProperties]}
  {emitterName=Blinking_Lights, rendererClasses [&NiagaraSpriteRendererProperties]}
  {emitterName=Light, rendererClasses [&NiagaraLightRendererProperties, &NiagaraSpriteRendererProperties]}
  {emitterName=Comet_Flame_Mesh, rendererClasses [&NiagaraMeshRendererProperties]}
 ]
 userVariables [
  User.bIsFrontEnd {type=NiagaraBool}
  User.bIsFrontEndPreview {type=NiagaraBool}
  User.BurstIndex {type=NiagaraFloat, defaultValue=1, description=""}
  User.Flame_Size {type=NiagaraFloat, defaultValue=2, description=""}
  User.Pos_Offset_X {type=NiagaraFloat, defaultValue=0, description=""}
  User.Pos_Offset_Y {type=NiagaraFloat, defaultValue=0, description=""}
  User.RibbonIDOffset {type=NiagaraInt32}
  User.SizeMult {type=NiagaraFloat, defaultValue=1, description=""}
  User.SwingSectionIndex {type=NiagaraInt32}
  User.Thruster_Height_Mult {type=NiagaraFloat, defaultValue=1, description=""}
  User.Trail_Life {type=NiagaraFloat, defaultValue=0, description=""}
  User.Variant {type=NiagaraFloat, defaultValue=0, description=""}
 ]
}
